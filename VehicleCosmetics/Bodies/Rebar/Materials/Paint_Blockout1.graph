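# Placeholder paint material for the Rebar vehicle body cosmetic, judging by its name and path, enabled for skeletal meshes and mesh deformers.
# It uses Default Lit shading with a small untextured graph of two scalar constants and two dark grey vector parameters.
# One grey is multiplied by a sharp, low-base Fresnel to form a view-angle rim term.
# The file does not show which material outputs these nodes drive.
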Material /VehicleCosmetics/Bodies/Rebar/Materials/Paint_Blockout1
graph {
 properties {bUsedWithMeshDeformer=true, bUsedWithSkeletalMesh=true, pixelDepthOffsetMode=PDOM_Legacy, usedShadingModels=Default Lit}
 nodes [
  Constant() {r=0.5}
  Constant() {r=0.2}
  Multiply(A=VectorParameter() {defaultValue=<rgba 0.0573 0.0573 0.0573 1>}.RGB, B=Fresnel() {baseReflectFraction=0.05, exponent=8})
  VectorParameter() {parameterName=Param_1, defaultValue=<rgba 0.0122 0.0122 0.0122 1>}
 ]
}
