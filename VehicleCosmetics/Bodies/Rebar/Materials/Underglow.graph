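# Material for the underglow of the Rebar vehicle body cosmetic, enabled for skeletal meshes and mesh deformers.
# Its small Default Lit graph multiplies a vector parameter's RGB, dark grey by default, by a constant to brighten it fivefold.
# The graph shows no link from that product to a material output, so the channel it drives is not shown.
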Material /VehicleCosmetics/Bodies/Rebar/Materials/Underglow
graph {
 properties {bUsedWithMeshDeformer=true, bUsedWithSkeletalMesh=true, pixelDepthOffsetMode=PDOM_Legacy, usedShadingModels=Default Lit}
 nodes [Multiply(A=VectorParameter() {defaultValue=<rgba 0.0226 0.0226 0.0226 1>}.RGB, B=Constant() {r=5})]
}
