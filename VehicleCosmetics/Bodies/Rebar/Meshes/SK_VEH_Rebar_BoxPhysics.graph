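# Simplified physics asset for the Rebar vehicle body cosmetic skeletal mesh, SK_VEH_Rebar.
# A single box shape named chassis_box on the root bone approximates the chassis, offset slightly along X and raised above the bone origin.
# It has no other bodies and no constraints, so the vehicle collides as one rigid box with no articulated parts.
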PhysicsAsset /VehicleCosmetics/Bodies/Rebar/Meshes/SK_VEH_Rebar_BoxPhysics
bodies {root [{shapeType=Box, extentX=444, extentY=216, extentZ=116, shapeName=chassis_box, center=<14.8957 0 85.2001>}]}
constraints []
