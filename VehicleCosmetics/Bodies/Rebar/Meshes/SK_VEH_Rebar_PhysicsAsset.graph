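# Physics asset for the Rebar vehicle body skeletal mesh, under the VehicleCosmetics bodies content.
# It contains a single body on the root bone, with no collision shapes listed, and no constraints.
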PhysicsAsset /VehicleCosmetics/Bodies/Rebar/Meshes/SK_VEH_Rebar_PhysicsAsset
bodies {root []}
constraints []
